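# Looping SoundCue for the DoubleDuty emote, pairing the main loop with scooter layers.
# A nested mixer blends three scooter loops (Lyr1, Lyr2, Lyr4), the latter two lowered, then a root mixer adds the main loop.
# Despite its foley name, it uses emote music class and concurrency, with emote attenuation and a send to the music audio-analysis submix.
# It keeps playing when silent, so the loops continue while inaudible.
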SoundCue /Game/Athena/Sounds/Emotes/DoubleDuty/EmoteFoley_DoubleDuty_FE_LP_Cue
root=Mixer_1
nodes [
 {id=Mixer_0, children [WavePlayer_1,WavePlayer_2,WavePlayer_3], properties {inputVolume [1,0.7,0.7]}}
 {id=Mixer_1, children [WavePlayer_0,Mixer_0], properties {inputVolume [1,1]}}
 {id=WavePlayer_0, properties {bLooping=true, soundWaveAssetPtr=&/IP_Emote_DoubleDuty_Main_LP}}
 {id=WavePlayer_1, properties {bLooping=true, soundWaveAssetPtr=&/IP_Emote_DoubleDuty_Scooter_LP_Lyr1}}
 {id=WavePlayer_2, properties {bLooping=true, soundWaveAssetPtr=&/IP_Emote_DoubleDuty_Scooter_LP_Lyr2}}
 {id=WavePlayer_3, properties {bLooping=true, soundWaveAssetPtr=&/IP_Emote_DoubleDuty_Scooter_LP_Lyr4}}
]
properties {
 attenuationSettings=&Fort_Emotes_Athena_Att
 maxDistance=1600
 soundClassObject=&Fort_Music_Emote
 virtualizationMode=PlayWhenSilent
 volumeMultiplier=1
 concurrencyOverrides {maxCount=3}
 concurrencySet [/Game/Sounds/Concurrency_Rules/Athena_EmoteMusic_Conc]
 soundSubmixSends [{minSendLevel=0, soundSubmix=&Fortnite_AudioAnalysis_Music}]
}
